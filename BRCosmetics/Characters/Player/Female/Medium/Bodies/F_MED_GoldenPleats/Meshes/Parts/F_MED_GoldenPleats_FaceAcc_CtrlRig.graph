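# Control Rig for the GoldenPleats female face accessory that reshapes collar bones as the head moves.
# Spherical pose readers on the head store four look-direction alphas, but only look-up is used, rotating the back collar bones.
# Two further inline readers weight translation and scale offsets on left and right collar bones.
# A chain of Spherical Collision functions between collar tips and head is present but not wired into execution.
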ControlRigBlueprint /BRCosmetics/Characters/Player/Female/Medium/Bodies/F_MED_GoldenPleats/Meshes/Parts/F_MED_GoldenPleats_FaceAcc_CtrlRig
graphs [
 ContainedGraph {nodes [
   n1 = RigVMFunctionEntryNode()
   SequenceExecution_2 = &RigVMFunction_Sequence(ExecuteContext=@n1.ExecuteContext)
   SequenceExecution_2_1 = &RigVMFunction_Sequence(ExecuteContext=@SequenceExecution_2.B)
   SequenceExecution_2_2 = &RigVMFunction_Sequence(ExecuteContext=@SequenceExecution_2_1.B)
   SequenceExecution_2_3 = &RigVMFunction_Sequence(ExecuteContext=@SequenceExecution_2_2.B)
   SequenceExecution_2_4 = &RigVMFunction_Sequence(ExecuteContext=@SequenceExecution_2_3.B)
   RigVMFunctionReturnNode(
    A=@SequenceExecution_2.A
    B=@SequenceExecution_2_1.A
    C=@SequenceExecution_2_2.A
    D=@SequenceExecution_2_3.A
    E=@SequenceExecution_2_4.A
    F=@SequenceExecution_2_4.B
   )
  ]}
 RigVMFunctionLibrary {links [], nodes []}
 RigVMModel {nodes [
   RigUnit_BeginExecution = &RigUnit_BeginExecution()
   SequenceExecution_2 = RigVMAggregateNode(ExecuteContext=@RigUnit_BeginExecution.ExecutePin)
   n1 = &RigUnit_SphericalPoseReader(
    ActiveRegionScaleFactors=(PositiveWidth=4,NegativeWidth=1,PositiveHeight=3,NegativeHeight=3)
    ActiveRegionSize=0.07
    Debug=(bDrawDebug=true,bDraw2D=False,bDrawLocalAxes=false,DebugScale=25,DebugSegments=20,DebugThickness=0.25)
    DriverAxis=(X=1,Y=0,Z=0)
    DriverItem=(Type=Bone,Name="head")
    FalloffRegionScaleFactors=(PositiveWidth=1,NegativeWidth=0.8,PositiveHeight=1,NegativeHeight=1)
    FalloffSize=0.4
    FlipHeightScaling=False
    FlipWidthScaling=False
    OptionalParentItem=(Type=Bone,Name="spine_05")
    RotationOffset=(X=70,Y=0,Z=-150)
    ExecutePin=@SequenceExecution_2.D
   )
   > RigVMVariableNode(Value=@n1.OutputParam) {variable=head_SPR_look_up_alpha}
   n2 = &RigUnit_SphericalPoseReader(
    ActiveRegionScaleFactors=(PositiveWidth=2,NegativeWidth=2,PositiveHeight=2.4,NegativeHeight=0.5)
    ActiveRegionSize=0.1
    Debug=(bDrawDebug=false,DebugScale=25,DebugSegments=20,DebugThickness=0.25)
    DriverAxis=(X=0,Y=1,Z=0)
    DriverItem=(Type=Bone,Name="head")
    FalloffRegionScaleFactors=(PositiveWidth=1,NegativeWidth=1,PositiveHeight=1,NegativeHeight=0.4)
    FalloffSize=0.25
    OptionalParentItem=(Type=Bone)
    RotationOffset=(X=120,Y=0,Z=0)
    ExecutePin=@SequenceExecution_2.B
   )
   > RigVMVariableNode(Value=@n2.OutputParam) {variable=head_SPR_look_left_alpha}
   n3 = RigVMVariableNode() {variable=head_SPR_look_up_alpha}
   &RigUnit_ModifyTransforms(
    ItemToModify="((Item=(Type=Bone,Name=\"dyn_collar_back_01\"),Transform=(Rotation=(X=0,Y=0,Z=-0.3007,W=0.9537),Translation=(X=0,Y=0,Z=0),Scale3D=(X=1,Y=1,Z=1))),(Item=(Type=Bone,Name=\"dyn_collar_back_02\"),Transform=(Rotation=(X=0,Y=0,Z=-0.0872,W=0.9962),Translation=(X=0,Y= ...(truncated)"
    WeightMinimum=0
    Weight=@n3.Value
    ExecutePin=@SequenceExecution_2.E
   )
   > &RigUnit_ModifyTransforms(
    ItemToModify="((Item=(Type=Bone,Name=\"dyn_collar_bk_r_01\"),Transform=(Rotation=(X=0,Y=0,Z=-0.2164,W=0.9763),Translation=(X=0,Y=0,Z=0),Scale3D=(X=1,Y=1,Z=1))),(Item=(Type=Bone,Name=\"dyn_collar_bk_r_02\"),Transform=(Rotation=(X=0,Y=0,Z=-0.0872,W=0.9962),Translation=(X=0,Y= ...(truncated)"
    WeightMinimum=0
    Weight=@n3.Value
   )
   > &RigUnit_ModifyTransforms(
    ItemToModify="((Item=(Type=Bone,Name=\"dyn_collar_bk_l_01\"),Transform=(Rotation=(X=0,Y=0,Z=-0.1305,W=0.9914),Translation=(X=0,Y=0,Z=0),Scale3D=(X=1,Y=1,Z=1))),(Item=(Type=Bone,Name=\"dyn_collar_bk_l_02\"),Transform=(Rotation=(X=0,Y=0,Z=-0.0436,W=0.999),Translation=(X=0,Y=0 ...(truncated)"
    WeightMinimum=0
    Weight=@n3.Value
   )
   > n4 = &RigUnit_SphericalPoseReader(
    ActiveRegionScaleFactors=(PositiveWidth=2,NegativeWidth=2,PositiveHeight=1,NegativeHeight=0.5)
    ActiveRegionSize=0.1
    Debug=(bDrawDebug=false,DebugScale=25,DebugSegments=20,DebugThickness=0.25)
    DriverAxis=(X=0,Y=1,Z=0)
    DriverItem=(Type=Bone,Name="head")
    FalloffRegionScaleFactors=(PositiveWidth=1,NegativeWidth=1,PositiveHeight=1,NegativeHeight=0.6)
    FalloffSize=0.25
    OptionalParentItem=(Type=Bone)
    RotationOffset=(X=145,Y=0,Z=0)
   )
   > &RigUnit_ModifyTransforms(
    ItemToModify="((Item=(Type=Bone,Name=\"dyn_collar_left_02\"),Transform=(Rotation=(X=0,Y=0,Z=0,W=1),Translation=(X=-1.5,Y=0,Z=0),Scale3D=(X=0.8,Y=1,Z=1))),(Item=(Type=Bone,Name=\"dyn_collar_bk_l_02\"),Transform=(Rotation=(X=0,Y=0,Z=0,W=1),Translation=(X=-0.5,Y=0,Z=0),Scale3D ...(truncated)"
    WeightMinimum=0
    Weight=@n4.OutputParam
   )
   > n5 = &RigUnit_SphericalPoseReader(
    ActiveRegionScaleFactors=(PositiveWidth=2,NegativeWidth=2,PositiveHeight=2.5,NegativeHeight=0.5)
    ActiveRegionSize=0.1
    Debug=(bDrawDebug=false,DebugScale=25,DebugSegments=20,DebugThickness=0.25)
    DriverAxis=(X=0,Y=1,Z=0)
    DriverItem=(Type=Bone,Name="head")
    FalloffRegionScaleFactors=(PositiveWidth=1,NegativeWidth=1,PositiveHeight=1.5,NegativeHeight=0.7)
    FalloffSize=0.25
    FlipHeightScaling=true
    FlipWidthScaling=false
    OptionalParentItem=(Type=Bone)
    RotationOffset=(X=35,Y=0,Z=0)
   )
   > &RigUnit_ModifyTransforms(
    ItemToModify="((Item=(Type=Bone,Name=\"dyn_collar_right_02\"),Transform=(Rotation=(X=0,Y=0,Z=0,W=1),Translation=(X=-1,Y=0,Z=0),Scale3D=(X=0.8,Y=1,Z=1))),(Item=(Type=Bone,Name=\"dyn_collar_bk_r_02\"),Transform=(Rotation=(X=0,Y=0,Z=0,W=1),Translation=(X=-0.5,Y=0,Z=0),Scale3D= ...(truncated)"
    WeightMinimum=0
    Weight=@n5.OutputParam
   )
   n6 = &RigUnit_SphericalPoseReader(
    ActiveRegionScaleFactors=(PositiveWidth=2,NegativeWidth=2,PositiveHeight=3,NegativeHeight=0.5)
    ActiveRegionSize=0.1
    Debug=(bDrawDebug=false,DebugScale=25,DebugSegments=20,DebugThickness=0.25)
    DriverAxis=(X=0,Y=1,Z=0)
    DriverItem=(Type=Bone,Name="head")
    FalloffRegionScaleFactors=(PositiveWidth=1,NegativeWidth=1,PositiveHeight=0,NegativeHeight=0.4)
    FalloffSize=0.27
    FlipHeightScaling=true
    FlipWidthScaling=false
    OptionalParentItem=(Type=Bone)
    RotationOffset=(X=60,Y=0,Z=0)
    ExecutePin=@SequenceExecution_2.C
   )
   > RigVMVariableNode(Value=@n6.OutputParam) {variable=head_SPR_look_right_alpha}
   n7 = &RigUnit_SphericalPoseReader(
    ActiveRegionScaleFactors=(PositiveWidth=2,NegativeWidth=0.5,PositiveHeight=2,NegativeHeight=2)
    ActiveRegionSize=0.075
    Debug=(bDrawDebug=false,DebugScale=25,DebugSegments=20,DebugThickness=0.25)
    DriverAxis=(X=0,Y=1,Z=0)
    DriverItem=(Type=Bone,Name="head")
    FalloffRegionScaleFactors=(PositiveWidth=1,NegativeWidth=0.4,PositiveHeight=1,NegativeHeight=1)
    FalloffSize=0.2
    FlipHeightScaling=false
    FlipWidthScaling=true
    OptionalParentItem=(Type=Bone)
    RotationOffset=(X=90,Y=0,Z=12.5)
    ExecutePin=@SequenceExecution_2.A
   )
   > RigVMVariableNode(Value=@n7.OutputParam) {variable=head_SPR_look_down_alpha}
   Spherical Collision = RigVMFunctionReferenceNode(
    Debug=false
    ExecuteContext=()
    Max Radius=8
    Min Radius=7
    Source=(Type=Bone,Name="dyn_collar_back_02")
    Source Offset=(X=0,Y=0,Z=0)
    Target=(Type=Bone,Name="head")
    Target Offset=(X=-2,Y=-3,Z=0)
    Weight=1
   )
   > Spherical Collision_1_3 = RigVMFunctionReferenceNode(
    Debug=false
    Max Radius=7
    Min Radius=5.5
    Source=(Type=Bone,Name="dyn_collar_bk_l_02")
    Source Offset=(X=0,Y=0,Z=0)
    Target=(Type=Bone,Name="head")
    Target Offset=(X=-1,Y=-2,Z=-1)
    Weight=1
   )
   > Spherical Collision_1_2_1 = RigVMFunctionReferenceNode(
    Debug=false
    Max Radius=7
    Min Radius=6
    Source=(Type=Bone,Name="dyn_collar_bk_r_02")
    Source Offset=(X=0,Y=0,Z=0)
    Target=(Type=Bone,Name="head")
    Target Offset=(X=-1,Y=-3,Z=2)
    Weight=1
   )
   > Spherical Collision_1 = RigVMFunctionReferenceNode(
    Debug=false
    Max Radius=8
    Min Radius=6
    Source=(Type=Bone,Name="dyn_collar_left_02")
    Source Offset=(X=0,Y=0,Z=0)
    Target=(Type=Bone,Name="head")
    Target Offset=(X=-1,Y=1,Z=-1)
    Weight=1
   )
   > Spherical Collision_1_1 = RigVMFunctionReferenceNode(
    Debug=false
    Max Radius=15
    Min Radius=6
    Source=(Type=Bone,Name="dyn_collar_fr_l_02")
    Source Offset=(X=0,Y=0,Z=0)
    Target=(Type=Bone,Name="head")
    Target Offset=(X=-3,Y=6,Z=-1)
    Weight=1
   )
   > Spherical Collision_1_2 = RigVMFunctionReferenceNode(
    Debug=false
    Max Radius=8
    Min Radius=6
    Source=(Type=Bone,Name="dyn_collar_right_02")
    Source Offset=(X=0,Y=0,Z=0)
    Target=(Type=Bone,Name="head")
    Target Offset=(X=-1,Y=1,Z=1)
    Weight=1
   )
   > Spherical Collision_1_1_1 = RigVMFunctionReferenceNode(
    Debug=false
    Max Radius=15
    Min Radius=7
    Source=(Type=Bone,Name="dyn_collar_fr_r_02")
    Source Offset=(X=0,Y=0,Z=0)
    Target=(Type=Bone,Name="head")
    Target Offset=(X=-1,Y=6,Z=1)
    Weight=1
   )
  ]}
]
hierarchy [
 root {loc=<0 0 0>, rot=<0 0 0 1>}
 pelvis {parent=root, loc=<0 -0.5301 98.0571>, rot=<0.0127 0.707 -0.0127 -0.707>}
 spine_01 {parent=pelvis, loc=<6.1208 -0.9328 0>, rot=<0 0 0.034 -0.9994>}
 spine_02 {parent=spine_01, loc=<6.5996 -0.2095 0>, rot=<0 0 0.0238 -0.9997>}
 spine_03 {parent=spine_02, loc=<6.4914 -0.1237 0>, rot=<0 0 0.0149 -0.9999>}
 spine_04 {parent=spine_03, loc=<6.9598 0.0794 0>, rot=<0 0 -0.0132 -0.9999>}
 spine_05 {parent=spine_04, loc=<6.9987 0 0>, rot=<0 0 -0.0015 -1>}
 neck_01 {parent=spine_05, loc=<11.3138 0.1529 0>, rot=<0 0 -0.2003 -0.9797>}
 neck_02 {parent=neck_01, loc=<4.5627 0 0>, rot=<0 0 0.0504 -0.9987>}
 head {parent=neck_02, loc=<5.3914 0 0>, rot=<0 0 0.115 -0.9934>}
 dyn_bangs_l_a_01 {parent=head, loc=<10.0534 7.1883 -2.9655>, rot=<0.047 -0.9192 0.0167 -0.3906>}
 dyn_bangs_l_a_02 {parent=dyn_bangs_l_a_01, loc=<3.1302 0 0>, rot=<0.039 -0.2718 0.0465 -0.9604>}
 dyn_bangs_l_a_03 {parent=dyn_bangs_l_a_02, loc=<5.0989 0 0>, rot=<0.0134 -0.1461 -0.0984 -0.9843>}
 dyn_bangs_l_a_04 {parent=dyn_bangs_l_a_03, loc=<4.6551 0 0>, rot=<-0.0073 -0.0859 -0.0755 -0.9934>}
 dyn_bangs_l_a_05 {parent=dyn_bangs_l_a_04, loc=<3.5518 0 0>, rot=<-0.0111 -0.0483 -0.0661 -0.9966>}
 dyn_bangs_l_b_01 {parent=head, loc=<10.6272 6.0592 -4.1508>, rot=<0.0911 -0.9601 0.0228 -0.2633>}
 dyn_bangs_l_b_02 {parent=dyn_bangs_l_b_01, loc=<5.4138 0 0>, rot=<0.036 -0.2283 -0.0237 -0.9726>}
 dyn_bangs_l_b_03 {parent=dyn_bangs_l_b_02, loc=<5.5783 0 0>, rot=<-0.0014 -0.067 -0.1528 -0.986>}
 dyn_bangs_l_b_04 {parent=dyn_bangs_l_b_03, loc=<4.4693 0 0>, rot=<0.0145 0.0519 -0.0969 -0.9938>}
 dyn_bangs_l_c_01 {parent=head, loc=<11.7947 4.2306 -3.1332>, rot=<0.0176 -0.921 0.0042 -0.3891>}
 dyn_bangs_l_c_02 {parent=dyn_bangs_l_c_01, loc=<4.1208 0 0>, rot=<0.0392 -0.295 0.0974 -0.9497>}
 dyn_bangs_l_c_03 {parent=dyn_bangs_l_c_02, loc=<4.98 0 0>, rot=<0.0171 -0.1326 -0.1046 -0.9855>}
 dyn_bangs_l_c_04 {parent=dyn_bangs_l_c_03, loc=<4.8642 0 0>, rot=<-0.0081 -0.0519 -0.1766 -0.9829>}
 dyn_bangs_l_c_05 {parent=dyn_bangs_l_c_04, loc=<4.2424 0 0>, rot=<-0.0359 -0.0848 -0.0678 -0.9934>}
 dyn_bangs_r_c_01 {parent=head, loc=<12.0052 4.1874 2.9482>, rot=<0.0281 -0.9383 -0.0075 0.3446>}
 dyn_bangs_r_c_02 {parent=dyn_bangs_r_c_01, loc=<5.5597 0 0>, rot=<-0.0398 0.2733 0.0891 -0.957>}
 dyn_bangs_r_c_03 {parent=dyn_bangs_r_c_02, loc=<4.5757 0 0>, rot=<-0.0116 0.1024 -0.1233 -0.987>}
 dyn_bangs_r_c_04 {parent=dyn_bangs_r_c_03, loc=<4.2938 0 0>, rot=<0.026 0.1657 -0.1422 -0.9755>}
 dyn_bangs_r_c_05 {parent=dyn_bangs_r_c_04, loc=<3.3367 0 0>, rot=<0.0399 0.0932 -0.1034 -0.9895>}
 dyn_bangs_r_b_01 {parent=head, loc=<11.2177 5.8748 3.6208>, rot=<0.0354 -0.9499 -0.009 0.3105>}
 dyn_bangs_r_b_02 {parent=dyn_bangs_r_b_01, loc=<4.7946 0 0>, rot=<-0.0362 0.2197 0.0946 -0.9703>}
 dyn_bangs_r_b_03 {parent=dyn_bangs_r_b_02, loc=<4.4153 0 0>, rot=<-0.0092 0.0781 -0.1428 -0.9866>}
 dyn_bangs_r_b_04 {parent=dyn_bangs_r_b_03, loc=<4.5095 0 0>, rot=<0.0056 0.0304 -0.1547 -0.9875>}
 dyn_bangs_r_a_01 {parent=head, loc=<9.8442 7.2542 3.5263>, rot=<0.0674 -0.9506 -0.0189 0.3026>}
 dyn_bangs_r_a_02 {parent=dyn_bangs_r_a_01, loc=<3.7986 0 0>, rot=<-0.0365 0.2097 0.0384 -0.9763>}
 dyn_bangs_r_a_03 {parent=dyn_bangs_r_a_02, loc=<3.8466 0 0>, rot=<-0.0125 0.099 -0.0858 -0.9913>}
 dyn_bangs_r_a_04 {parent=dyn_bangs_r_a_03, loc=<4.2109 0 0>, rot=<0.0122 0.1006 -0.1587 -0.9821>}
 dyn_bangs_r_a_05 {parent=dyn_bangs_r_a_04, loc=<4.1174 0 0>, rot=<0.0067 0.0181 -0.0735 -0.9971>}
 dyn_hair_l_MSTR {parent=head, loc=<10.3596 -4.7199 -5.5016>, rot=<0.0038 -0.9225 0.386 -0.0016>}
 dyn_hair_l {parent=dyn_hair_l_MSTR, loc=<0 0 0>, rot=<0 0 0 1>}
 dyn_hair_l_16 {parent=dyn_hair_l, loc=<5.2409 1.9684 8.0943>, rot=<0.3789 0.2158 -0.111 -0.8931>}
 dyn_hair_l_15 {parent=dyn_hair_l, loc=<4.3885 -1.5638 6.7222>, rot=<-0.1669 0.1497 0.084 -0.9709>}
 dyn_hair_l_14 {parent=dyn_hair_l, loc=<4.7257 1.2534 3.6576>, rot=<0.7308 0.0234 -0.1196 -0.6716>}
 dyn_hair_l_13 {parent=dyn_hair_l, loc=<8.0809 1.03 6.2225>, rot=<0.3839 0.0919 -0.0355 -0.9181>}
 dyn_hair_l_12 {parent=dyn_hair_l, loc=<7.6018 2.7274 6.4908>, rot=<0.3419 0.185 -0.1175 -0.9138>}
 dyn_hair_l_11 {parent=dyn_hair_l, loc=<5.4656 4.0923 6.6078>, rot=<0.2462 0.2978 -0.2851 -0.8772>}
 dyn_hair_l_10 {parent=dyn_hair_l, loc=<2.9259 2.5492 7.8837>, rot=<0.1642 0.4594 -0.0617 -0.8707>}
 dyn_hair_l_09 {parent=dyn_hair_l, loc=<4.5823 0.4879 8.5694>, rot=<-0.0017 0.1736 -0.1013 -0.9796>}
 dyn_hair_l_08 {parent=dyn_hair_l, loc=<7.2068 -0.4195 7.6721>, rot=<-0.0179 0.2364 0.0061 -0.9715>}
 dyn_hair_l_07 {parent=dyn_hair_l, loc=<7.6244 -2.2221 5.22>, rot=<-0.432 -0.0084 0.0926 -0.897>}
 dyn_hair_l_06 {parent=dyn_hair_l, loc=<4.6257 -2.0301 2.9551>, rot=<-0.831 0.0508 0.1749 -0.5257>}
 dyn_hair_l_05 {parent=dyn_hair_l, loc=<2.6006 -2.8969 4.7255>, rot=<-0.7409 0.1632 0.0834 -0.6461>}
 dyn_hair_l_04 {parent=dyn_hair_l, loc=<0.7048 -2.403 5.6555>, rot=<-0.4641 0.2837 0.0609 -0.8369>}
 dyn_hair_l_03 {parent=dyn_hair_l, loc=<-0.3672 -0.1212 7.3298>, rot=<-0.1132 0.6608 0.0149 -0.7418>}
 dyn_hair_l_02 {parent=dyn_hair_l, loc=<-0.8845 0.5045 6.4269>, rot=<0.2875 0.5524 -0.189 -0.7592>}
 dyn_hair_l_01 {parent=dyn_hair_l, loc=<2.6281 4.1512 7.035>, rot=<0.5293 0.4538 -0.3168 -0.643>}
 dyn_hair_r_MSTR {parent=head, loc=<10.3596 -4.7199 5.5016>, rot=<0.0016 -0.3863 0.9224 -0.0038>}
 dyn_hair_r {parent=dyn_hair_r_MSTR, loc=<0 0 0>, rot=<0 0 0 1>}
 clavicle_l {parent=spine_05, loc=<5.4831 -0.1053 -3.2946>, rot=<0.0465 -0.7289 -0.009 -0.6829>}
 upperarm_l {parent=clavicle_l, loc=<9.8658 -1.2136 -0.4105>, rot=<0.0164 -0.3245 -0.0269 -0.9454>}
 upperarm_twist_01_l {parent=upperarm_l, loc=<7.2135 -0.1134 0>, rot=<0 0 0 1>}
 deform_upperarm_bk_base_l {parent=upperarm_twist_01_l, loc=<-6.4053 -4.3746 0.314>, rot=<0 0 0 -1>}
 deform_upperarm_bk_l {parent=deform_upperarm_bk_base_l, loc=<0 0 0>, rot=<0 0 0 1>}
 deform_upperarm_fr_base_l {parent=upperarm_twist_01_l, loc=<-5.6312 5.092 0.3354>, rot=<0 0 0 -1>}
 deform_upperarm_fr_l {parent=deform_upperarm_fr_base_l, loc=<0 0 0>, rot=<0 0 0 1>}
 deform_upperarm_up_base_l {parent=upperarm_twist_01_l, loc=<-8.891 0.4825 3.9517>, rot=<0 0 0 -1>}
 deform_upperarm_up_l {parent=deform_upperarm_up_base_l, loc=<0 0 0>, rot=<0 0 0 1>}
 clavicle_r {parent=spine_05, loc=<5.4831 -0.1053 3.2946>, rot=<0.7289 0.0465 -0.6829 0.009>}
 upperarm_r {parent=clavicle_r, loc=<-9.8658 1.2136 0.4105>, rot=<0.0164 -0.3245 -0.0269 -0.9454>}
 dyn_collar_fr_l_mstr {parent=spine_05, loc=<7.7516 4.269 -4.5>, rot=<0.9992 0.0401 0 0>}
 dyn_collar_fr_l_01 {parent=dyn_collar_fr_l_mstr, loc=<0 0 0>, rot=<0 0 0 1>}
 dyn_collar_fr_l_02 {parent=dyn_collar_fr_l_01, loc=<2.7526 -1.2403 0>, rot=<0.3822 0 0 -0.9241>}
 dyn_collar_left_mstr {parent=spine_05, loc=<9.8038 0.8579 -6.4999>, rot=<-0.6829 -0.1118 -0.1095 -0.7135>}
 dyn_collar_left_01 {parent=dyn_collar_left_mstr, loc=<0 0 0>, rot=<0 0 0 1>}
 dyn_collar_left_02 {parent=dyn_collar_left_01, loc=<3.242 0 0>, rot=<0 0 0 1>}
 dyn_collar_bk_l_mstr {parent=spine_05, loc=<10.6895 -2.541 -4.614>, rot=<-0.3824 -0.0153 -0.037 -0.9231>}
 dyn_collar_bk_l_01 {parent=dyn_collar_bk_l_mstr, loc=<0 0 0>, rot=<0 0 0 1>}
 dyn_collar_bk_l_02 {parent=dyn_collar_bk_l_01, loc=<3.242 0.6153 -0.6153>, rot=<0 0 0 1>}
 dyn_collar_back_mstr {parent=spine_05, loc=<10.6969 -4.3945 -0.0296>, rot=<0 0 -0.0401 -0.9992>}
 dyn_collar_back_01 {parent=dyn_collar_back_mstr, loc=<0 0 0>, rot=<0 0 0 1>}
 dyn_collar_back_02 {parent=dyn_collar_back_01, loc=<3.242 0.329 0>, rot=<0 0 0 1>}
 dyn_collar_bk_r_mstr {parent=spine_05, loc=<10.6895 -2.541 4.614>, rot=<0.3824 0.0153 -0.037 -0.9231>}
 dyn_collar_bk_r_01 {parent=dyn_collar_bk_r_mstr, loc=<0 0 0>, rot=<0 0 0 1>}
 dyn_collar_bk_r_02 {parent=dyn_collar_bk_r_01, loc=<3.242 0.6153 0.6153>, rot=<0 0 0 1>}
 dyn_collar_right_mstr {parent=spine_05, loc=<9.8038 0.8579 6.5>, rot=<0.7135 0.1095 -0.1118 -0.6829>}
 dyn_collar_right_01 {parent=dyn_collar_right_mstr, loc=<0 0 0>, rot=<0 0 0 1>}
 dyn_collar_right_02 {parent=dyn_collar_right_01, loc=<3.242 0 0>, rot=<0 0 0 1>}
 dyn_collar_fr_r_mstr {parent=spine_05, loc=<7.7516 4.269 4.5>, rot=<0.9992 0.0401 0 0>}
 dyn_collar_fr_r_01 {parent=dyn_collar_fr_r_mstr, loc=<0 0 0>, rot=<0 0 0 1>}
 dyn_collar_fr_r_02 {parent=dyn_collar_fr_r_01, loc=<2.7526 -1.2403 0>, rot=<-0.3831 0 0 -0.9237>}
 dyn_collar_front_mstr {parent=spine_05, loc=<5.7766 6.5535 -0.0296>, rot=<0.9992 0.0401 0 0>}
 dyn_collar_front_01 {parent=dyn_collar_front_mstr, loc=<0 0 0>, rot=<0 0 0 1>}
 dyn_collar_front_02 {parent=dyn_collar_front_01, loc=<3.187 -1.394 0>, rot=<0 0 0 1>}
]
variables [
 head_SPR_look_down_alpha {type=double, default=0}
 head_SPR_look_left_alpha {type=double, default=0}
 head_SPR_look_right_alpha {type=double, default=0}
 head_SPR_look_up_alpha {type=double, default=0}
]
