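# Physics asset for the face accessory part of the female medium GoldenPleats player cosmetic, covering hair, bangs and collar bones with capsule bodies.
# Six bangs chains, three per side, link segment to segment up to the head with swing and twist limits, while each side hair body hangs from a free-rotating master body on the head.
# Head, neck, clavicle and upper arm capsules cover the surrounding body, and the eight collar bodies have no constraints.
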PhysicsAsset /BRCosmetics/Characters/Player/Female/Medium/Bodies/F_MED_GoldenPleats/Meshes/Parts/F_MED_GoldenPleats_FaceAcc_Physics
bodies {
 clavicle_l [{
   shapeType=Capsule
   length=4.7081
   radius=5.3397
   shapeName=clavicle_l_capsule
   center=<6.5243 -0.6953 0.104>
   rotation=<rot 16.9211 87.287 -90>
  }]
 clavicle_r [{
   shapeType=Capsule
   length=5.4909
   radius=5.7737
   shapeName=clavicle_r_capsule
   center=<-6.4341 0.3237 0.8599>
   rotation=<rot 3.2384 98.0696 -90>
  }]
 dyn_bangs_l_a_01 [{shapeType=Capsule, length=2, radius=1, shapeName=dyn_bangs_l_a_01_capsule, center=<2 0 0>, rotation=<rot 90 0 0>}]
 dyn_bangs_l_a_02 [{shapeType=Capsule, length=2, radius=1, shapeName=dyn_bangs_l_a_02_capsule, center=<2 0 0>, rotation=<rot 90 0 0>}]
 dyn_bangs_l_a_03 [{shapeType=Capsule, length=2, radius=1, shapeName=dyn_bangs_l_a_03_capsule, center=<2 0 0>, rotation=<rot 90 0 0>}]
 dyn_bangs_l_a_04 [{shapeType=Capsule, length=2, radius=1, shapeName=dyn_bangs_l_a_04_capsule, center=<2 0 0>, rotation=<rot 90 0 0>}]
 dyn_bangs_l_a_05 [{shapeType=Capsule, length=1.875, radius=1.125, shapeName=dyn_bangs_l_a_05_capsule, center=<1.5617 0 0>, rotation=<rot 90 0 0>}]
 dyn_bangs_l_b_01 [{shapeType=Capsule, length=2, radius=1, shapeName=dyn_bangs_l_b_01_capsule, center=<2 0 0>, rotation=<rot 90 0 0>}]
 dyn_bangs_l_b_02 [{shapeType=Capsule, length=2, radius=1, shapeName=dyn_bangs_l_b_02_capsule, center=<2 0 0>, rotation=<rot 90 0 0>}]
 dyn_bangs_l_b_03 [{shapeType=Capsule, length=2, radius=1, shapeName=dyn_bangs_l_b_03_capsule, center=<2 0 0>, rotation=<rot 90 0 0>}]
 dyn_bangs_l_b_04 [{shapeType=Capsule, length=1.5, radius=1.125, shapeName=dyn_bangs_l_b_04_capsule, center=<1.5871 0 0>, rotation=<rot 90 0 0>}]
 dyn_bangs_l_c_01 [{shapeType=Capsule, length=2, radius=1, shapeName=dyn_bangs_l_c_01_capsule, center=<2 0 0>, rotation=<rot 90 0 0>}]
 dyn_bangs_l_c_02 [{shapeType=Capsule, length=2.5, radius=1.125, shapeName=dyn_bangs_l_c_02_capsule, center=<2.5 0 0>, rotation=<rot 90 0 0>}]
 dyn_bangs_l_c_03 [{shapeType=Capsule, length=2.75, radius=1, shapeName=dyn_bangs_l_c_03_capsule, center=<2 0 0>, rotation=<rot 90 0 0>}]
 dyn_bangs_l_c_04 [{shapeType=Capsule, length=2, radius=1, shapeName=dyn_bangs_l_c_04_capsule, center=<2 0 0>, rotation=<rot 90 0 0>}]
 dyn_bangs_l_c_05 [{shapeType=Capsule, length=1.125, radius=1.125, shapeName=dyn_bangs_l_c_05_capsule, center=<1.3956 0 0>, rotation=<rot 90 0 0>}]
 dyn_bangs_r_a_01 [{shapeType=Capsule, length=2, radius=1, shapeName=dyn_bangs_r_a_01_capsule, center=<2 0 0>, rotation=<rot 90 0 0>}]
 dyn_bangs_r_a_02 [{shapeType=Capsule, length=2, radius=1, shapeName=dyn_bangs_r_a_02_capsule, center=<2 0 0>, rotation=<rot 90 0 0>}]
 dyn_bangs_r_a_03 [{shapeType=Capsule, length=2, radius=1, shapeName=dyn_bangs_r_a_03_capsule, center=<2 0 0>, rotation=<rot 90 0 0>}]
 dyn_bangs_r_a_04 [{shapeType=Capsule, length=2, radius=1, shapeName=dyn_bangs_r_a_04_capsule, center=<2 0 0>, rotation=<rot 90 0 0>}]
 dyn_bangs_r_a_05 [{shapeType=Capsule, length=1.875, radius=1.125, shapeName=dyn_bangs_r_a_05_capsule, center=<2 0 0>, rotation=<rot 90 0 0>}]
 dyn_bangs_r_b_01 [{shapeType=Capsule, length=2, radius=1, shapeName=dyn_bangs_r_b_01_capsule, center=<2 0 0>, rotation=<rot 90 0 0>}]
 dyn_bangs_r_b_02 [{shapeType=Capsule, length=2, radius=1, shapeName=dyn_bangs_r_b_02_capsule, center=<2 0 0>, rotation=<rot 90 0 0>}]
 dyn_bangs_r_b_03 [{shapeType=Capsule, length=2, radius=1, shapeName=dyn_bangs_r_b_03_capsule, center=<2 0 0>, rotation=<rot 90 0 0>}]
 dyn_bangs_r_b_04 [{shapeType=Capsule, length=3.125, radius=1.125, shapeName=dyn_bangs_r_b_04_capsule, center=<3 0 0>, rotation=<rot 90 0 0>}]
 dyn_bangs_r_c_01 [{shapeType=Capsule, length=2, radius=1, shapeName=dyn_bangs_r_c_01_capsule, center=<2 0 0>, rotation=<rot 90 0 0>}]
 dyn_bangs_r_c_02 [{shapeType=Capsule, length=2, radius=1, shapeName=dyn_bangs_r_c_02_capsule, center=<2 0 0>, rotation=<rot 90 0 0>}]
 dyn_bangs_r_c_03 [{shapeType=Capsule, length=2, radius=1, shapeName=dyn_bangs_r_c_03_capsule, center=<2 0 0>, rotation=<rot 90 0 0>}]
 dyn_bangs_r_c_04 [{shapeType=Capsule, length=2, radius=1, shapeName=dyn_bangs_r_c_04_capsule, center=<2 0 0>, rotation=<rot 90 0 0>}]
 dyn_bangs_r_c_05 [{shapeType=Capsule, length=1.875, radius=1.125, shapeName=dyn_bangs_r_c_05_capsule, center=<2 0 0>, rotation=<rot 90 0 0>}]
 dyn_collar_back_02 [{shapeType=Capsule, length=4.875, radius=3.125, shapeName=dyn_collar_back_02_capsule, center=<-0.6583 1.6428 0>, rotation=<rot 90 0 0>}]
 dyn_collar_bk_l_02 [{shapeType=Capsule, length=4.25, radius=3.375, shapeName=dyn_collar_bk_r_02_capsule, center=<-0.8486 0.4283 0>, rotation=<rot 90 0 0>}]
 dyn_collar_bk_r_02 [{shapeType=Capsule, length=4.25, radius=3.375, shapeName=dyn_collar_bk_r_02_capsule, center=<-0.8486 0.4283 0>, rotation=<rot 90 0 0>}]
 dyn_collar_fr_l_02 [{shapeType=Capsule, length=2.875, radius=3.75, shapeName=dyn_collar_fr_r_02_capsule, center=<-0.4076 0.5918 0>, rotation=<rot 90 0 14.9999>}]
 dyn_collar_fr_r_02 [{shapeType=Capsule, length=2.875, radius=3.75, shapeName=dyn_collar_fr_r_02_capsule, center=<-0.2446 0.5481 0>, rotation=<rot 90 0 14.9999>}]
 dyn_collar_front_02 [{
   shapeType=Capsule
   length=4.75
   radius=3.25
   shapeName=dyn_collar_front_02_capsule
   center=<-1.3932 1.717 0>
   rotation=<rot 90 180 -155.0001>
  }]
 dyn_collar_left_02 [{shapeType=Capsule, length=3.875, radius=3.75, shapeName=dyn_collar_right_02_capsule, center=<-1.4774 0.0958 0>, rotation=<rot 90 0 0>}]
 dyn_collar_right_02 [{shapeType=Capsule, length=3.875, radius=3.75, shapeName=dyn_collar_right_02_capsule, center=<-0.8321 0.0958 0>, rotation=<rot 90 0 0>}]
 dyn_hair_l [{shapeType=Capsule, length=6.1817, radius=4.5111, shapeName=dyn_hair_l_capsule, center=<4 1 6>, rotation=<rot 0 93.4921 -79.1654>}]
 dyn_hair_l_MSTR [{shapeType=Capsule, length=0.505, radius=0.505, shapeName=dyn_hair_l_MSTR_capsule, rotation=<rot 0 -90 0>}]
 dyn_hair_r [{shapeType=Capsule, length=6.1817, radius=4.5111, shapeName=dyn_hair_l_capsule, center=<4 -1 6>, rotation=<rot 0 93.4921 -79.1654>}]
 dyn_hair_r_MSTR [{shapeType=Capsule, length=0.505, radius=0.505, shapeName=dyn_hair_l_MSTR_capsule, rotation=<rot 0 -90 0>}]
 head [
  {shapeType=Capsule, length=3.4546, radius=6.3843, shapeName=head_capsule, center=<2.3025 0.855 0>, rotation=<rot 90 0 5.0001>}
  {shapeType=Capsule, length=3.125, radius=5.5, shapeName=head_capsule, center=<2.2477 4.1477 0>, rotation=<rot -90 160.5288 -160.5289>}
  {shapeType=Capsule, length=1.875, radius=2.125, shapeName=head_capsule, center=<0.9028 -0.9422 -5.6539>, rotation=<rot -90 -0.0001 -0.0001>}
  {shapeType=Capsule, length=1.875, radius=2.125, shapeName=head_capsule, center=<1 -0.9422 5.218>, rotation=<rot 90 0 0>}
 ]
 neck_01 [{shapeType=Capsule, length=8.255, radius=6.63, shapeName=neck_01_capsule, center=<-5.2979 2.0717 0>, rotation=<rot 0 -85.0003 90>}]
 upperarm_l [{shapeType=Capsule, length=13.4229, radius=4.1429, shapeName=upperarm_l_capsule, center=<5 -1.25 0.0001>, rotation=<rot 0 -90 90>}]
 upperarm_r [{shapeType=Capsule, length=13.0479, radius=4.5179, shapeName=upperarm_r_capsule, center=<-5 1.25 0.0001>, rotation=<rot 0 -90 90>}]
}
constraints [
 {bone1Name=dyn_bangs_l_a_01, bone2Name=head, swing1LimitDegrees=45, swing2LimitDegrees=45, twistLimitDegrees=15}
 {bone1Name=dyn_bangs_l_a_02, bone2Name=dyn_bangs_l_a_01, swing1LimitDegrees=45, swing2LimitDegrees=45, twistLimitDegrees=15}
 {bone1Name=dyn_bangs_l_a_03, bone2Name=dyn_bangs_l_a_02, swing1LimitDegrees=45, swing2LimitDegrees=45, twistLimitDegrees=15}
 {bone1Name=dyn_bangs_l_a_04, bone2Name=dyn_bangs_l_a_03, swing1LimitDegrees=45, swing2LimitDegrees=45, twistLimitDegrees=15}
 {bone1Name=dyn_bangs_l_a_05, bone2Name=dyn_bangs_l_a_04, swing1LimitDegrees=45, swing2LimitDegrees=45, twistLimitDegrees=15}
 {bone1Name=dyn_bangs_l_b_01, bone2Name=head, swing1LimitDegrees=45, swing2LimitDegrees=45, twistLimitDegrees=15}
 {bone1Name=dyn_bangs_l_b_02, bone2Name=dyn_bangs_l_b_01, swing1LimitDegrees=45, swing2LimitDegrees=45, twistLimitDegrees=15}
 {bone1Name=dyn_bangs_l_b_03, bone2Name=dyn_bangs_l_b_02, swing1LimitDegrees=45, swing2LimitDegrees=45, twistLimitDegrees=15}
 {bone1Name=dyn_bangs_l_b_04, bone2Name=dyn_bangs_l_b_03, swing1LimitDegrees=45, swing2LimitDegrees=45, twistLimitDegrees=15}
 {bone1Name=dyn_bangs_l_c_01, bone2Name=head, swing1LimitDegrees=45, swing2LimitDegrees=45, twistLimitDegrees=15}
 {bone1Name=dyn_bangs_l_c_02, bone2Name=dyn_bangs_l_c_01, swing1LimitDegrees=45, swing2LimitDegrees=45, twistLimitDegrees=15}
 {bone1Name=dyn_bangs_l_c_03, bone2Name=dyn_bangs_l_c_02, swing1LimitDegrees=45, swing2LimitDegrees=45, twistLimitDegrees=15}
 {bone1Name=dyn_bangs_l_c_04, bone2Name=dyn_bangs_l_c_03, swing1LimitDegrees=45, swing2LimitDegrees=45, twistLimitDegrees=15}
 {bone1Name=dyn_bangs_l_c_05, bone2Name=dyn_bangs_l_c_04, swing1LimitDegrees=45, swing2LimitDegrees=45, twistLimitDegrees=15}
 {bone1Name=dyn_bangs_r_a_01, bone2Name=head, swing1LimitDegrees=45, swing2LimitDegrees=45, twistLimitDegrees=15}
 {bone1Name=dyn_bangs_r_a_02, bone2Name=dyn_bangs_r_a_01, swing1LimitDegrees=45, swing2LimitDegrees=45, twistLimitDegrees=15}
 {bone1Name=dyn_bangs_r_a_03, bone2Name=dyn_bangs_r_a_02, swing1LimitDegrees=45, swing2LimitDegrees=45, twistLimitDegrees=15}
 {bone1Name=dyn_bangs_r_a_04, bone2Name=dyn_bangs_r_a_03, swing1LimitDegrees=45, swing2LimitDegrees=45, twistLimitDegrees=15}
 {bone1Name=dyn_bangs_r_a_05, bone2Name=dyn_bangs_r_a_04, swing1LimitDegrees=45, swing2LimitDegrees=45, twistLimitDegrees=15}
 {bone1Name=dyn_bangs_r_b_01, bone2Name=head, swing1LimitDegrees=45, swing2LimitDegrees=45, twistLimitDegrees=15}
 {bone1Name=dyn_bangs_r_b_02, bone2Name=dyn_bangs_r_b_01, swing1LimitDegrees=45, swing2LimitDegrees=45, twistLimitDegrees=15}
 {bone1Name=dyn_bangs_r_b_03, bone2Name=dyn_bangs_r_b_02, swing1LimitDegrees=45, swing2LimitDegrees=45, twistLimitDegrees=15}
 {bone1Name=dyn_bangs_r_b_04, bone2Name=dyn_bangs_r_b_03, swing1LimitDegrees=45, swing2LimitDegrees=45, twistLimitDegrees=15}
 {bone1Name=dyn_bangs_r_c_01, bone2Name=head, swing1LimitDegrees=45, swing2LimitDegrees=45, twistLimitDegrees=15}
 {bone1Name=dyn_bangs_r_c_02, bone2Name=dyn_bangs_r_c_01, swing1LimitDegrees=45, swing2LimitDegrees=45, twistLimitDegrees=15}
 {bone1Name=dyn_bangs_r_c_03, bone2Name=dyn_bangs_r_c_02, swing1LimitDegrees=45, swing2LimitDegrees=45, twistLimitDegrees=15}
 {bone1Name=dyn_bangs_r_c_04, bone2Name=dyn_bangs_r_c_03, swing1LimitDegrees=45, swing2LimitDegrees=45, twistLimitDegrees=15}
 {bone1Name=dyn_bangs_r_c_05, bone2Name=dyn_bangs_r_c_04, swing1LimitDegrees=45, swing2LimitDegrees=45, twistLimitDegrees=15}
 {
  bone1Name=head
  bone2Name=neck_01
  swing1LimitDegrees=45
  swing1Motion=Free
  swing2LimitDegrees=45
  swing2Motion=Free
  twistLimitDegrees=45
  twistMotion=Free
 }
 {bone1Name=dyn_hair_l, bone2Name=dyn_hair_l_MSTR, swing1LimitDegrees=60, swing2LimitDegrees=60, twistLimitDegrees=15}
 {
  bone1Name=dyn_hair_l_MSTR
  bone2Name=head
  swing1LimitDegrees=45
  swing1Motion=Free
  swing2LimitDegrees=45
  swing2Motion=Free
  twistLimitDegrees=45
  twistMotion=Free
 }
 {
  bone1Name=dyn_hair_r_MSTR
  bone2Name=head
  swing1LimitDegrees=45
  swing1Motion=Free
  swing2LimitDegrees=45
  swing2Motion=Free
  twistLimitDegrees=45
  twistMotion=Free
 }
 {bone1Name=dyn_hair_r, bone2Name=dyn_hair_r_MSTR, swing1LimitDegrees=60, swing2LimitDegrees=60, twistLimitDegrees=15}
 {
  bone1Name=upperarm_r
  bone2Name=clavicle_r
  swing1LimitDegrees=45
  swing1Motion=Free
  swing2LimitDegrees=45
  swing2Motion=Free
  twistLimitDegrees=45
  twistMotion=Free
 }
 {
  bone1Name=upperarm_l
  bone2Name=clavicle_l
  swing1LimitDegrees=45
  swing1Motion=Free
  swing2LimitDegrees=45
  swing2Motion=Free
  twistLimitDegrees=45
  twistMotion=Free
 }
]
