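# Animation Blueprint for the GoldenPleats female character's face accessory, adding secondary motion to the pose copied from its attached parent.
# That pose runs through a Control Rig, a spine_05-space RigidBody simulation, then serial single-bone Anim Dynamics springs on left and right hair bones.
# A deprecated gravity-override function updates the RigidBody's gravity every frame; hair nodes use fixed gravity.
# A layered blend and several duplicate hair nodes sit disconnected from the output.
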AnimBlueprint /BRCosmetics/Characters/Player/Female/Medium/Bodies/F_MED_GoldenPleats/Meshes/Parts/F_MED_GoldenPleats_FaceAcc_AnimBP
target_skeleton=&/F_MED_GoldenPleats_FaceAcc_Skeleton
graphs [
 AnimGraph {nodes [
   n1 = Anim Dynamics - Bone: dyn_hair_l_15(Alpha=1, ExternalForce="0,0,0") {
    angularBiasOverride=0.2
    angularDampingOverride=0.99
    angularSpringConstant=14000
    bAngularSpring=true
    bOverrideAngularBias=true
    bOverrideAngularDamping=true
    bUseGravityOverride=true
    gravityScale=0.35
    alphaBoolBlend {blendOption=Linear}
    boundBone {boneName=dyn_hair_l_15}
    componentAppliedLinearAccClamp=<200000 200000 200000>
    componentLinearAccScale=<1 1 1>
    componentLinearVelScale=<1 1 1>
    gravityOverride {z=-3000}
    physicsBodyDefinitions [{
      boundBone {boneName=dyn_hair_l_15}
      boxExtents=<40 40 40>
      constraintSetup {angularLimitsMax=<7.5 7.5 7.5>, angularLimitsMin=<-7.5 -7.5 -7.5>, angularTarget=<1 0 0>}
      localJointOffset=<3 2 0>
     }]
    relativeSpaceBone {boneName=head}
   }
   Anim Dynamics - Bone: dyn_hair_l_16(ComponentPose=@n1, Alpha=1, ExternalForce="0,0,0") {
    angularBiasOverride=0.2
    angularDampingOverride=0.99
    angularSpringConstant=14000
    bAngularSpring=true
    bOverrideAngularBias=true
    bOverrideAngularDamping=true
    bUseGravityOverride=true
    gravityScale=0.35
    alphaBoolBlend {blendOption=Linear}
    boundBone {boneName=dyn_hair_l_16}
    componentAppliedLinearAccClamp=<200000 200000 200000>
    componentLinearAccScale=<1 1 1>
    componentLinearVelScale=<1 1 1>
    gravityOverride {z=-3000}
    physicsBodyDefinitions [{
      boundBone {boneName=dyn_hair_l_16}
      boxExtents=<40 40 40>
      constraintSetup {angularLimitsMax=<7.5 7.5 7.5>, angularLimitsMin=<-7.5 -7.5 -7.5>, angularTarget=<1 0 0>}
      localJointOffset=<2 1.3835 2.143>
     }]
    relativeSpaceBone {boneName=head}
   }
   Anim Dynamics - Bone: dyn_hair_l_01(Alpha=1, ExternalForce="0,0,0") {
    angularBiasOverride=0.2
    angularDampingOverride=0.99
    angularSpringConstant=14000
    bAngularSpring=true
    bOverrideAngularBias=true
    bOverrideAngularDamping=true
    bUseGravityOverride=true
    gravityScale=0.35
    alphaBoolBlend {blendOption=Linear}
    boundBone {boneName=dyn_hair_l_01}
    componentAppliedLinearAccClamp=<200000 200000 200000>
    componentLinearAccScale=<1 1 1>
    componentLinearVelScale=<1 1 1>
    gravityOverride {z=-3000}
    physicsBodyDefinitions [{
      boundBone {boneName=dyn_hair_l_01}
      boxExtents=<40 40 40>
      constraintSetup {angularLimitsMax=<7.5 7.5 7.5>, angularLimitsMin=<-7.5 -7.5 -7.5>, angularTarget=<1 0 0>}
      localJointOffset=<0.4827 2.8969 1.037>
     }]
    relativeSpaceBone {boneName=head}
   }
   Layered blend per bone(BlendPoses_0=Copy Pose From Mesh(SourceMeshComponent=Get Skel_Mesh()) {bUseAttachedParent=true}) {layerSetup [{branchFilters [
       {boneName=upperarm_l}
       {boneName=upperarm_r}
       {boneName=dyn_collar_right_02}
       {boneName=dyn_collar_back_02}
       {boneName=dyn_collar_left_02}
       {boneName=None}
      ]}]}
   n2 = Control Rig(Source=Copy Pose From Mesh() {bUseAttachedParent=true}) {
    alphaBoolBlend {blendOption=Linear}
    controlRigAssetReference {blueprintRigClass=&/F_MED_GoldenPleats_FaceAcc_CtrlRig.F_MED_GoldenPleats_FaceAcc_CtrlRig_C}
   }
   n3 = RigidBody!(
    ComponentPose=Local To Component(LocalPose=@n2)
    Alpha
    OverrideWorldGravity:Vector=Get GravityOverride()
    ExternalForce
    ComponentLinearAccScale:Vector="0.3,0.3,0.5"
    ComponentLinearVelScale:Vector="0.3,0.3,0.5"
   ) -> (Pose) {
    bClampLinearTranslationLimitToRefPose=true
    bOverrideWorldGravity=true
    overridePhysicsAsset=&/F_MED_GoldenPleats_FaceAcc_Physics
    simulationSpace=BaseBoneSpace
    alphaBoolBlend {blendOption=Linear}
    baseBoneRef {boneName=spine_05}
    componentAppliedLinearAccClamp=<2500 2500 2500>
    componentLinearAccScale=<0.3 0.3 0.5>
    componentLinearVelScale=<0.3 0.3 0.5>
    overrideWorldGravity {z=-800}
   }
   n4 = Anim Dynamics - Bone: dyn_hair_l_01(ComponentPose=@n3, Alpha=1, ExternalForce="0,0,0") {
    angularBiasOverride=0.2
    angularDampingOverride=0.99
    angularSpringConstant=14000
    bAngularSpring=true
    bOverrideAngularBias=true
    bOverrideAngularDamping=true
    bUseGravityOverride=true
    gravityScale=0.35
    alphaBoolBlend {blendOption=Linear}
    boundBone {boneName=dyn_hair_l_01}
    componentAppliedLinearAccClamp=<200000 200000 200000>
    componentLinearAccScale=<1 1 1>
    componentLinearVelScale=<1 1 1>
    gravityOverride {z=-3000}
    physicsBodyDefinitions [{
      boundBone {boneName=dyn_hair_l_01}
      boxExtents=<40 40 40>
      constraintSetup {angularLimitsMax=<7.5 7.5 7.5>, angularLimitsMin=<-7.5 -7.5 -7.5>, angularTarget=<1 0 0>}
      localJointOffset=<1.6649 0.068 0.093>
     }]
    relativeSpaceBone {boneName=head}
   }
   n5 = Anim Dynamics - Bone: dyn_hair_l_02(ComponentPose=@n4, Alpha=1, ExternalForce="0,0,0") {
    angularBiasOverride=0.2
    angularDampingOverride=0.99
    angularSpringConstant=14000
    bAngularSpring=true
    bOverrideAngularBias=true
    bOverrideAngularDamping=true
    bUseGravityOverride=true
    gravityScale=0.35
    alphaBoolBlend {blendOption=Linear}
    boundBone {boneName=dyn_hair_l_02}
    componentAppliedLinearAccClamp=<200000 200000 200000>
    componentLinearAccScale=<1 1 1>
    componentLinearVelScale=<1 1 1>
    gravityOverride {z=-3000}
    physicsBodyDefinitions [{
      boundBone {boneName=dyn_hair_l_02}
      boxExtents=<40 40 40>
      constraintSetup {angularLimitsMax=<7.5 7.5 7.5>, angularLimitsMin=<-7.5 -7.5 -7.5>, angularTarget=<1 0 0>}
      localJointOffset=<2.7554 -1.1705 1.105>
     }]
    relativeSpaceBone {boneName=head}
   }
   n6 = Anim Dynamics - Bone: dyn_hair_l_03(ComponentPose=@n5, Alpha=1, ExternalForce="0,0,0") {
    angularBiasOverride=0.2
    angularDampingOverride=0.99
    angularSpringConstant=14000
    bAngularSpring=true
    bOverrideAngularBias=true
    bOverrideAngularDamping=true
    bUseGravityOverride=true
    gravityScale=0.35
    alphaBoolBlend {blendOption=Linear}
    boundBone {boneName=dyn_hair_l_03}
    componentAppliedLinearAccClamp=<200000 200000 200000>
    componentLinearAccScale=<1 1 1>
    componentLinearVelScale=<1 1 1>
    gravityOverride {z=-3000}
    physicsBodyDefinitions [{
      boundBone {boneName=dyn_hair_l_03}
      boxExtents=<40 40 40>
      constraintSetup {angularLimitsMax=<7.5 7.5 7.5>, angularLimitsMin=<-7.5 -7.5 -7.5>, angularTarget=<1 0 0>}
      localJointOffset=<1.1607 -2.6905 0.457>
     }]
    relativeSpaceBone {boneName=head}
   }
   n7 = Anim Dynamics - Bone: dyn_hair_l_04(ComponentPose=@n6, Alpha=1, ExternalForce="0,0,0") {
    angularBiasOverride=0.2
    angularDampingOverride=0.99
    angularSpringConstant=14000
    bAngularSpring=true
    bOverrideAngularBias=true
    bOverrideAngularDamping=true
    bUseGravityOverride=true
    gravityScale=0.35
    alphaBoolBlend {blendOption=Linear}
    boundBone {boneName=dyn_hair_l_04}
    componentAppliedLinearAccClamp=<200000 200000 200000>
    componentLinearAccScale=<1 1 1>
    componentLinearVelScale=<1 1 1>
    gravityOverride {z=-3000}
    physicsBodyDefinitions [{
      boundBone {boneName=dyn_hair_l_04}
      boxExtents=<40 40 40>
      constraintSetup {angularLimitsMax=<7.5 7.5 7.5>, angularLimitsMin=<-7.5 -7.5 -7.5>, angularTarget=<1 0 0>}
      localJointOffset=<1.495 0.5196 -0.7997>
     }]
    relativeSpaceBone {boneName=head}
   }
   n8 = Anim Dynamics - Bone: dyn_hair_l_05(ComponentPose=@n7, Alpha=1, ExternalForce="0,0,0") {
    angularBiasOverride=0.2
    angularDampingOverride=0.99
    angularSpringConstant=14000
    bAngularSpring=true
    bOverrideAngularBias=true
    bOverrideAngularDamping=true
    bUseGravityOverride=true
    gravityScale=0.35
    alphaBoolBlend {blendOption=Linear}
    boundBone {boneName=dyn_hair_l_05}
    componentAppliedLinearAccClamp=<200000 200000 200000>
    componentLinearAccScale=<1 1 1>
    componentLinearVelScale=<1 1 1>
    gravityOverride {z=-3000}
    physicsBodyDefinitions [{
      boundBone {boneName=dyn_hair_l_05}
      boxExtents=<40 40 40>
      constraintSetup {angularLimitsMax=<7.5 7.5 7.5>, angularLimitsMin=<-7.5 -7.5 -7.5>, angularTarget=<1 0 0>}
      localJointOffset=<0.75 0.8534 1.6045>
     }]
    relativeSpaceBone {boneName=head}
   }
   n9 = Anim Dynamics - Bone: dyn_hair_l_06(ComponentPose=@n8, Alpha=1, ExternalForce="0,0,0") {
    angularBiasOverride=0.2
    angularDampingOverride=0.99
    angularSpringConstant=14000
    bAngularSpring=true
    bOverrideAngularBias=true
    bOverrideAngularDamping=true
    bUseGravityOverride=true
    gravityScale=0.35
    alphaBoolBlend {blendOption=Linear}
    boundBone {boneName=dyn_hair_l_06}
    componentAppliedLinearAccClamp=<200000 200000 200000>
    componentLinearAccScale=<1 1 1>
    componentLinearVelScale=<1 1 1>
    gravityOverride {z=-3000}
    physicsBodyDefinitions [{
      boundBone {boneName=dyn_hair_l_06}
      boxExtents=<40 40 40>
      constraintSetup {angularLimitsMax=<7.5 7.5 7.5>, angularLimitsMin=<-7.5 -7.5 -7.5>, angularTarget=<1 0 0>}
      localJointOffset=<0.7747 2.7737 1.1391>
     }]
    relativeSpaceBone {boneName=head}
   }
   n10 = Anim Dynamics - Bone: dyn_hair_l_07(ComponentPose=@n9, Alpha=1, ExternalForce="0,0,0") {
    angularBiasOverride=0.2
    angularDampingOverride=0.99
    angularSpringConstant=14000
    bAngularSpring=true
    bOverrideAngularBias=true
    bOverrideAngularDamping=true
    bUseGravityOverride=true
    gravityScale=0.35
    alphaBoolBlend {blendOption=Linear}
    boundBone {boneName=dyn_hair_l_07}
    componentAppliedLinearAccClamp=<200000 200000 200000>
    componentLinearAccScale=<1 1 1>
    componentLinearVelScale=<1 1 1>
    gravityOverride {z=-3000}
    physicsBodyDefinitions [{
      boundBone {boneName=dyn_hair_l_07}
      boxExtents=<40 40 40>
      constraintSetup {angularLimitsMax=<7.5 7.5 7.5>, angularLimitsMin=<-7.5 -7.5 -7.5>, angularTarget=<1 0 0>}
      localJointOffset=<0.9493 1.4429 2.2624>
     }]
    relativeSpaceBone {boneName=head}
   }
   n11 = Anim Dynamics - Bone: dyn_hair_l_08(ComponentPose=@n10, Alpha=1, ExternalForce="0,0,0") {
    angularBiasOverride=0.2
    angularDampingOverride=0.99
    angularSpringConstant=14000
    bAngularSpring=true
    bOverrideAngularBias=true
    bOverrideAngularDamping=true
    bUseGravityOverride=true
    gravityScale=0.35
    alphaBoolBlend {blendOption=Linear}
    boundBone {boneName=dyn_hair_l_08}
    componentAppliedLinearAccClamp=<200000 200000 200000>
    componentLinearAccScale=<1 1 1>
    componentLinearVelScale=<1 1 1>
    gravityOverride {z=-3000}
    physicsBodyDefinitions [{
      boundBone {boneName=dyn_hair_l_08}
      boxExtents=<40 40 40>
      constraintSetup {angularLimitsMax=<7.5 7.5 7.5>, angularLimitsMin=<-7.5 -7.5 -7.5>, angularTarget=<1 0 0>}
      localJointOffset=<4 0.5 0.5>
     }]
    relativeSpaceBone {boneName=head}
   }
   n12 = Anim Dynamics - Bone: dyn_hair_l_09(ComponentPose=@n11, Alpha=1, ExternalForce="0,0,0") {
    angularBiasOverride=0.2
    angularDampingOverride=0.99
    angularSpringConstant=14000
    bAngularSpring=true
    bOverrideAngularBias=true
    bOverrideAngularDamping=true
    bUseGravityOverride=true
    gravityScale=0.35
    alphaBoolBlend {blendOption=Linear}
    boundBone {boneName=dyn_hair_l_09}
    componentAppliedLinearAccClamp=<200000 200000 200000>
    componentLinearAccScale=<1 1 1>
    componentLinearVelScale=<1 1 1>
    gravityOverride {z=-3000}
    physicsBodyDefinitions [{
      boundBone {boneName=dyn_hair_l_09}
      boxExtents=<40 40 40>
      constraintSetup {angularLimitsMax=<7.5 7.5 7.5>, angularLimitsMin=<-7.5 -7.5 -7.5>, angularTarget=<1 0 0>}
      localJointOffset=<2.0441 -0.0305 1.5432>
     }]
    relativeSpaceBone {boneName=head}
   }
   n13 = Anim Dynamics - Bone: dyn_hair_l_10(ComponentPose=@n12, Alpha=1, ExternalForce="0,0,0") {
    angularBiasOverride=0.2
    angularDampingOverride=0.99
    angularSpringConstant=14000
    bAngularSpring=true
    bOverrideAngularBias=true
    bOverrideAngularDamping=true
    bUseGravityOverride=true
    gravityScale=0.35
    alphaBoolBlend {blendOption=Linear}
    boundBone {boneName=dyn_hair_l_10}
    componentAppliedLinearAccClamp=<200000 200000 200000>
    componentLinearAccScale=<1 1 1>
    componentLinearVelScale=<1 1 1>
    gravityOverride {z=-3000}
    physicsBodyDefinitions [{
      boundBone {boneName=dyn_hair_l_10}
      boxExtents=<40 40 40>
      constraintSetup {angularLimitsMax=<7.5 7.5 7.5>, angularLimitsMin=<-7.5 -7.5 -7.5>, angularTarget=<1 0 0>}
      localJointOffset=<2.4801 2.2397 1.6312>
     }]
    relativeSpaceBone {boneName=head}
   }
   n14 = Anim Dynamics - Bone: dyn_hair_l_11(ComponentPose=@n13, Alpha=1, ExternalForce="0,0,0") {
    angularBiasOverride=0.2
    angularDampingOverride=0.99
    angularSpringConstant=14000
    bAngularSpring=true
    bOverrideAngularBias=true
    bOverrideAngularDamping=true
    bUseGravityOverride=true
    gravityScale=0.35
    alphaBoolBlend {blendOption=Linear}
    boundBone {boneName=dyn_hair_l_11}
    componentAppliedLinearAccClamp=<200000 200000 200000>
    componentLinearAccScale=<1 1 1>
    componentLinearVelScale=<1 1 1>
    gravityOverride {z=-3000}
    physicsBodyDefinitions [{
      boundBone {boneName=dyn_hair_l_11}
      boxExtents=<40 40 40>
      constraintSetup {angularLimitsMax=<7.5 7.5 7.5>, angularLimitsMin=<-7.5 -7.5 -7.5>, angularTarget=<1 0 0>}
      localJointOffset=<0.2587 -0.6529 1.3511>
     }]
    relativeSpaceBone {boneName=head}
   }
   n15 = Anim Dynamics - Bone: dyn_hair_l_12(ComponentPose=@n14, Alpha=1, ExternalForce="0,0,0") {
    angularBiasOverride=0.2
    angularDampingOverride=0.99
    angularSpringConstant=14000
    bAngularSpring=true
    bOverrideAngularBias=true
    bOverrideAngularDamping=true
    bUseGravityOverride=true
    gravityScale=0.35
    alphaBoolBlend {blendOption=Linear}
    boundBone {boneName=dyn_hair_l_12}
    componentAppliedLinearAccClamp=<200000 200000 200000>
    componentLinearAccScale=<1 1 1>
    componentLinearVelScale=<1 1 1>
    gravityOverride {z=-3000}
    physicsBodyDefinitions [{
      boundBone {boneName=dyn_hair_l_12}
      boxExtents=<40 40 40>
      constraintSetup {angularLimitsMax=<7.5 7.5 7.5>, angularLimitsMin=<-7.5 -7.5 -7.5>, angularTarget=<1 0 0>}
      localJointOffset=<0.8545 1.0907 -0.8031>
     }]
    relativeSpaceBone {boneName=head}
   }
   n16 = Anim Dynamics - Bone: dyn_hair_l_13(ComponentPose=@n15, Alpha=1, ExternalForce="0,0,0") {
    angularBiasOverride=0.2
    angularDampingOverride=0.99
    angularSpringConstant=14000
    bAngularSpring=true
    bOverrideAngularBias=true
    bOverrideAngularDamping=true
    bUseGravityOverride=true
    gravityScale=0.35
    alphaBoolBlend {blendOption=Linear}
    boundBone {boneName=dyn_hair_l_13}
    componentAppliedLinearAccClamp=<200000 200000 200000>
    componentLinearAccScale=<1 1 1>
    componentLinearVelScale=<1 1 1>
    gravityOverride {z=-3000}
    physicsBodyDefinitions [{
      boundBone {boneName=dyn_hair_l_13}
      boxExtents=<40 40 40>
      constraintSetup {angularLimitsMax=<7.5 7.5 7.5>, angularLimitsMin=<-7.5 -7.5 -7.5>, angularTarget=<1 0 0>}
      localJointOffset=<1.9715 0.7465 0.7959>
     }]
    relativeSpaceBone {boneName=head}
   }
   n17 = Anim Dynamics - Bone: dyn_hair_l_14(ComponentPose=@n16, Alpha=1, ExternalForce="0,0,0") {
    angularBiasOverride=0.2
    angularDampingOverride=0.99
    angularSpringConstant=14000
    bAngularSpring=true
    bOverrideAngularBias=true
    bOverrideAngularDamping=true
    bUseGravityOverride=true
    gravityScale=0.35
    alphaBoolBlend {blendOption=Linear}
    boundBone {boneName=dyn_hair_l_14}
    componentAppliedLinearAccClamp=<200000 200000 200000>
    componentLinearAccScale=<1 1 1>
    componentLinearVelScale=<1 1 1>
    gravityOverride {z=-3000}
    physicsBodyDefinitions [{
      boundBone {boneName=dyn_hair_l_14}
      boxExtents=<40 40 40>
      constraintSetup {angularLimitsMax=<7.5 7.5 7.5>, angularLimitsMin=<-7.5 -7.5 -7.5>, angularTarget=<1 0 0>}
      localJointOffset=<2.2127 0.1441 2.9271>
     }]
    relativeSpaceBone {boneName=head}
   }
   n18 = Anim Dynamics - Bone: dyn_hair_l_15(ComponentPose=@n17, Alpha=1, ExternalForce="0,0,0") {
    angularBiasOverride=0.2
    angularDampingOverride=0.99
    angularSpringConstant=14000
    bAngularSpring=true
    bOverrideAngularBias=true
    bOverrideAngularDamping=true
    bUseGravityOverride=true
    gravityScale=0.35
    alphaBoolBlend {blendOption=Linear}
    boundBone {boneName=dyn_hair_l_15}
    componentAppliedLinearAccClamp=<200000 200000 200000>
    componentLinearAccScale=<1 1 1>
    componentLinearVelScale=<1 1 1>
    gravityOverride {z=-3000}
    physicsBodyDefinitions [{
      boundBone {boneName=dyn_hair_l_15}
      boxExtents=<40 40 40>
      constraintSetup {angularLimitsMax=<7.5 7.5 7.5>, angularLimitsMin=<-7.5 -7.5 -7.5>, angularTarget=<1 0 0>}
      localJointOffset=<3 2 0>
     }]
    relativeSpaceBone {boneName=head}
   }
   n19 = Anim Dynamics - Bone: dyn_hair_l_16(ComponentPose=@n18, Alpha=1, ExternalForce="0,0,0") {
    angularBiasOverride=0.2
    angularDampingOverride=0.99
    angularSpringConstant=14000
    bAngularSpring=true
    bOverrideAngularBias=true
    bOverrideAngularDamping=true
    bUseGravityOverride=true
    gravityScale=0.35
    alphaBoolBlend {blendOption=Linear}
    boundBone {boneName=dyn_hair_l_16}
    componentAppliedLinearAccClamp=<200000 200000 200000>
    componentLinearAccScale=<1 1 1>
    componentLinearVelScale=<1 1 1>
    gravityOverride {z=-3000}
    physicsBodyDefinitions [{
      boundBone {boneName=dyn_hair_l_16}
      boxExtents=<40 40 40>
      constraintSetup {angularLimitsMax=<7.5 7.5 7.5>, angularLimitsMin=<-7.5 -7.5 -7.5>, angularTarget=<1 0 0>}
      localJointOffset=<2 1.3835 2.143>
     }]
    relativeSpaceBone {boneName=head}
   }
   n20 = Anim Dynamics - Bone: dyn_hair_r_01(ComponentPose=@n19, Alpha=1, ExternalForce="0,0,0") {
    angularBiasOverride=0.2
    angularDampingOverride=0.99
    angularSpringConstant=14000
    bAngularSpring=true
    bOverrideAngularBias=true
    bOverrideAngularDamping=true
    bUseGravityOverride=true
    gravityScale=0.35
    alphaBoolBlend {blendOption=Linear}
    boundBone {boneName=dyn_hair_r_01}
    componentAppliedLinearAccClamp=<200000 200000 200000>
    componentLinearAccScale=<1 1 1>
    componentLinearVelScale=<1 1 1>
    gravityOverride {z=-3000}
    physicsBodyDefinitions [{
      boundBone {boneName=dyn_hair_r_01}
      boxExtents=<40 40 40>
      constraintSetup {angularLimitsMax=<7.5 7.5 7.5>, angularLimitsMin=<-7.5 -7.5 -7.5>, angularTarget=<1 0 0>}
      localJointOffset=<1.6649 0.068 0.093>
     }]
    relativeSpaceBone {boneName=head}
   }
   n21 = Anim Dynamics - Bone: dyn_hair_r_02(ComponentPose=@n20, Alpha=1, ExternalForce="0,0,0") {
    angularBiasOverride=0.2
    angularDampingOverride=0.99
    angularSpringConstant=14000
    bAngularSpring=true
    bOverrideAngularBias=true
    bOverrideAngularDamping=true
    bUseGravityOverride=true
    gravityScale=0.35
    alphaBoolBlend {blendOption=Linear}
    boundBone {boneName=dyn_hair_r_02}
    componentAppliedLinearAccClamp=<200000 200000 200000>
    componentLinearAccScale=<1 1 1>
    componentLinearVelScale=<1 1 1>
    gravityOverride {z=-3000}
    physicsBodyDefinitions [{
      boundBone {boneName=dyn_hair_r_02}
      boxExtents=<40 40 40>
      constraintSetup {angularLimitsMax=<7.5 7.5 7.5>, angularLimitsMin=<-7.5 -7.5 -7.5>, angularTarget=<1 0 0>}
      localJointOffset=<2.7554 -1.1705 1.105>
     }]
    relativeSpaceBone {boneName=head}
   }
   n22 = Anim Dynamics - Bone: dyn_hair_r_03(ComponentPose=@n21, Alpha=1, ExternalForce="0,0,0") {
    angularBiasOverride=0.2
    angularDampingOverride=0.99
    angularSpringConstant=14000
    bAngularSpring=true
    bOverrideAngularBias=true
    bOverrideAngularDamping=true
    bUseGravityOverride=true
    gravityScale=0.35
    alphaBoolBlend {blendOption=Linear}
    boundBone {boneName=dyn_hair_r_03}
    componentAppliedLinearAccClamp=<200000 200000 200000>
    componentLinearAccScale=<1 1 1>
    componentLinearVelScale=<1 1 1>
    gravityOverride {z=-3000}
    physicsBodyDefinitions [{
      boundBone {boneName=dyn_hair_r_03}
      boxExtents=<40 40 40>
      constraintSetup {angularLimitsMax=<7.5 7.5 7.5>, angularLimitsMin=<-7.5 -7.5 -7.5>, angularTarget=<1 0 0>}
      localJointOffset=<1.1607 -2.6905 0.457>
     }]
    relativeSpaceBone {boneName=head}
   }
   n23 = Anim Dynamics - Bone: dyn_hair_r_04(ComponentPose=@n22, Alpha=1, ExternalForce="0,0,0") {
    angularBiasOverride=0.2
    angularDampingOverride=0.99
    angularSpringConstant=14000
    bAngularSpring=true
    bOverrideAngularBias=true
    bOverrideAngularDamping=true
    bUseGravityOverride=true
    gravityScale=0.35
    alphaBoolBlend {blendOption=Linear}
    boundBone {boneName=dyn_hair_r_04}
    componentAppliedLinearAccClamp=<200000 200000 200000>
    componentLinearAccScale=<1 1 1>
    componentLinearVelScale=<1 1 1>
    gravityOverride {z=-3000}
    physicsBodyDefinitions [{
      boundBone {boneName=dyn_hair_r_04}
      boxExtents=<40 40 40>
      constraintSetup {angularLimitsMax=<7.5 7.5 7.5>, angularLimitsMin=<-7.5 -7.5 -7.5>, angularTarget=<1 0 0>}
      localJointOffset=<1.495 0.5196 -0.7997>
     }]
    relativeSpaceBone {boneName=head}
   }
   n24 = Anim Dynamics - Bone: dyn_hair_r_05(ComponentPose=@n23, Alpha=1, ExternalForce="0,0,0") {
    angularBiasOverride=0.2
    angularDampingOverride=0.99
    angularSpringConstant=14000
    bAngularSpring=true
    bOverrideAngularBias=true
    bOverrideAngularDamping=true
    bUseGravityOverride=true
    gravityScale=0.35
    alphaBoolBlend {blendOption=Linear}
    boundBone {boneName=dyn_hair_r_05}
    componentAppliedLinearAccClamp=<200000 200000 200000>
    componentLinearAccScale=<1 1 1>
    componentLinearVelScale=<1 1 1>
    gravityOverride {z=-3000}
    physicsBodyDefinitions [{
      boundBone {boneName=dyn_hair_r_05}
      boxExtents=<40 40 40>
      constraintSetup {angularLimitsMax=<7.5 7.5 7.5>, angularLimitsMin=<-7.5 -7.5 -7.5>, angularTarget=<1 0 0>}
      localJointOffset=<0.75 0.8534 1.6045>
     }]
    relativeSpaceBone {boneName=head}
   }
   n25 = Anim Dynamics - Bone: dyn_hair_r_06(ComponentPose=@n24, Alpha=1, ExternalForce="0,0,0") {
    angularBiasOverride=0.2
    angularDampingOverride=0.99
    angularSpringConstant=14000
    bAngularSpring=true
    bOverrideAngularBias=true
    bOverrideAngularDamping=true
    bUseGravityOverride=true
    gravityScale=0.35
    alphaBoolBlend {blendOption=Linear}
    boundBone {boneName=dyn_hair_r_06}
    componentAppliedLinearAccClamp=<200000 200000 200000>
    componentLinearAccScale=<1 1 1>
    componentLinearVelScale=<1 1 1>
    gravityOverride {z=-3000}
    physicsBodyDefinitions [{
      boundBone {boneName=dyn_hair_r_06}
      boxExtents=<40 40 40>
      constraintSetup {angularLimitsMax=<7.5 7.5 7.5>, angularLimitsMin=<-7.5 -7.5 -7.5>, angularTarget=<1 0 0>}
      localJointOffset=<0.7747 2.7737 1.1391>
     }]
    relativeSpaceBone {boneName=head}
   }
   n26 = Anim Dynamics - Bone: dyn_hair_r_07(ComponentPose=@n25, Alpha=1, ExternalForce="0,0,0") {
    angularBiasOverride=0.2
    angularDampingOverride=0.99
    angularSpringConstant=14000
    bAngularSpring=true
    bOverrideAngularBias=true
    bOverrideAngularDamping=true
    bUseGravityOverride=true
    gravityScale=0.35
    alphaBoolBlend {blendOption=Linear}
    boundBone {boneName=dyn_hair_r_07}
    componentAppliedLinearAccClamp=<200000 200000 200000>
    componentLinearAccScale=<1 1 1>
    componentLinearVelScale=<1 1 1>
    gravityOverride {z=-3000}
    physicsBodyDefinitions [{
      boundBone {boneName=dyn_hair_r_07}
      boxExtents=<40 40 40>
      constraintSetup {angularLimitsMax=<7.5 7.5 7.5>, angularLimitsMin=<-7.5 -7.5 -7.5>, angularTarget=<1 0 0>}
      localJointOffset=<0.9493 1.4429 2.2624>
     }]
    relativeSpaceBone {boneName=head}
   }
   n27 = Anim Dynamics - Bone: dyn_hair_r_08(ComponentPose=@n26, Alpha=1, ExternalForce="0,0,0") {
    angularBiasOverride=0.2
    angularDampingOverride=0.99
    angularSpringConstant=14000
    bAngularSpring=true
    bOverrideAngularBias=true
    bOverrideAngularDamping=true
    bUseGravityOverride=true
    gravityScale=0.35
    alphaBoolBlend {blendOption=Linear}
    boundBone {boneName=dyn_hair_r_08}
    componentAppliedLinearAccClamp=<200000 200000 200000>
    componentLinearAccScale=<1 1 1>
    componentLinearVelScale=<1 1 1>
    gravityOverride {z=-3000}
    physicsBodyDefinitions [{
      boundBone {boneName=dyn_hair_r_08}
      boxExtents=<40 40 40>
      constraintSetup {angularLimitsMax=<7.5 7.5 7.5>, angularLimitsMin=<-7.5 -7.5 -7.5>, angularTarget=<1 0 0>}
      localJointOffset=<4 0.5 0.5>
     }]
    relativeSpaceBone {boneName=head}
   }
   n28 = Anim Dynamics - Bone: dyn_hair_r_09(ComponentPose=@n27, Alpha=1, ExternalForce="0,0,0") {
    angularBiasOverride=0.2
    angularDampingOverride=0.99
    angularSpringConstant=14000
    bAngularSpring=true
    bOverrideAngularBias=true
    bOverrideAngularDamping=true
    bUseGravityOverride=true
    gravityScale=0.35
    alphaBoolBlend {blendOption=Linear}
    boundBone {boneName=dyn_hair_r_09}
    componentAppliedLinearAccClamp=<200000 200000 200000>
    componentLinearAccScale=<1 1 1>
    componentLinearVelScale=<1 1 1>
    gravityOverride {z=-3000}
    physicsBodyDefinitions [{
      boundBone {boneName=dyn_hair_r_09}
      boxExtents=<40 40 40>
      constraintSetup {angularLimitsMax=<7.5 7.5 7.5>, angularLimitsMin=<-7.5 -7.5 -7.5>, angularTarget=<1 0 0>}
      localJointOffset=<2.0441 -0.0305 1.5432>
     }]
    relativeSpaceBone {boneName=head}
   }
   n29 = Anim Dynamics - Bone: dyn_hair_r_10(ComponentPose=@n28, Alpha=1, ExternalForce="0,0,0") {
    angularBiasOverride=0.2
    angularDampingOverride=0.99
    angularSpringConstant=14000
    bAngularSpring=true
    bOverrideAngularBias=true
    bOverrideAngularDamping=true
    bUseGravityOverride=true
    gravityScale=0.35
    alphaBoolBlend {blendOption=Linear}
    boundBone {boneName=dyn_hair_r_10}
    componentAppliedLinearAccClamp=<200000 200000 200000>
    componentLinearAccScale=<1 1 1>
    componentLinearVelScale=<1 1 1>
    gravityOverride {z=-3000}
    physicsBodyDefinitions [{
      boundBone {boneName=dyn_hair_r_10}
      boxExtents=<40 40 40>
      constraintSetup {angularLimitsMax=<7.5 7.5 7.5>, angularLimitsMin=<-7.5 -7.5 -7.5>, angularTarget=<1 0 0>}
      localJointOffset=<2.4801 2.2397 1.6312>
     }]
    relativeSpaceBone {boneName=head}
   }
   n30 = Anim Dynamics - Bone: dyn_hair_r_11(ComponentPose=@n29, Alpha=1, ExternalForce="0,0,0") {
    angularBiasOverride=0.2
    angularDampingOverride=0.99
    angularSpringConstant=14000
    bAngularSpring=true
    bOverrideAngularBias=true
    bOverrideAngularDamping=true
    bUseGravityOverride=true
    gravityScale=0.35
    alphaBoolBlend {blendOption=Linear}
    boundBone {boneName=dyn_hair_r_11}
    componentAppliedLinearAccClamp=<200000 200000 200000>
    componentLinearAccScale=<1 1 1>
    componentLinearVelScale=<1 1 1>
    gravityOverride {z=-3000}
    physicsBodyDefinitions [{
      boundBone {boneName=dyn_hair_r_11}
      boxExtents=<40 40 40>
      constraintSetup {angularLimitsMax=<7.5 7.5 7.5>, angularLimitsMin=<-7.5 -7.5 -7.5>, angularTarget=<1 0 0>}
      localJointOffset=<0.2587 -0.6529 1.3511>
     }]
    relativeSpaceBone {boneName=head}
   }
   n31 = Anim Dynamics - Bone: dyn_hair_r_12(ComponentPose=@n30, Alpha=1, ExternalForce="0,0,0") {
    angularBiasOverride=0.2
    angularDampingOverride=0.99
    angularSpringConstant=14000
    bAngularSpring=true
    bOverrideAngularBias=true
    bOverrideAngularDamping=true
    bUseGravityOverride=true
    gravityScale=0.35
    alphaBoolBlend {blendOption=Linear}
    boundBone {boneName=dyn_hair_r_12}
    componentAppliedLinearAccClamp=<200000 200000 200000>
    componentLinearAccScale=<1 1 1>
    componentLinearVelScale=<1 1 1>
    gravityOverride {z=-3000}
    physicsBodyDefinitions [{
      boundBone {boneName=dyn_hair_r_12}
      boxExtents=<40 40 40>
      constraintSetup {angularLimitsMax=<7.5 7.5 7.5>, angularLimitsMin=<-7.5 -7.5 -7.5>, angularTarget=<1 0 0>}
      localJointOffset=<0.8545 1.0907 -0.8031>
     }]
    relativeSpaceBone {boneName=head}
   }
   n32 = Anim Dynamics - Bone: dyn_hair_r_13(ComponentPose=@n31, Alpha=1, ExternalForce="0,0,0") {
    angularBiasOverride=0.2
    angularDampingOverride=0.99
    angularSpringConstant=14000
    bAngularSpring=true
    bOverrideAngularBias=true
    bOverrideAngularDamping=true
    bUseGravityOverride=true
    gravityScale=0.35
    alphaBoolBlend {blendOption=Linear}
    boundBone {boneName=dyn_hair_r_13}
    componentAppliedLinearAccClamp=<200000 200000 200000>
    componentLinearAccScale=<1 1 1>
    componentLinearVelScale=<1 1 1>
    gravityOverride {z=-3000}
    physicsBodyDefinitions [{
      boundBone {boneName=dyn_hair_r_13}
      boxExtents=<40 40 40>
      constraintSetup {angularLimitsMax=<7.5 7.5 7.5>, angularLimitsMin=<-7.5 -7.5 -7.5>, angularTarget=<1 0 0>}
      localJointOffset=<1.9715 0.7465 0.7959>
     }]
    relativeSpaceBone {boneName=head}
   }
   n33 = Anim Dynamics - Bone: dyn_hair_r_14(ComponentPose=@n32, Alpha=1, ExternalForce="0,0,0") {
    angularBiasOverride=0.2
    angularDampingOverride=0.99
    angularSpringConstant=14000
    bAngularSpring=true
    bOverrideAngularBias=true
    bOverrideAngularDamping=true
    bUseGravityOverride=true
    gravityScale=0.35
    alphaBoolBlend {blendOption=Linear}
    boundBone {boneName=dyn_hair_r_14}
    componentAppliedLinearAccClamp=<200000 200000 200000>
    componentLinearAccScale=<1 1 1>
    componentLinearVelScale=<1 1 1>
    gravityOverride {z=-3000}
    physicsBodyDefinitions [{
      boundBone {boneName=dyn_hair_r_14}
      boxExtents=<40 40 40>
      constraintSetup {angularLimitsMax=<7.5 7.5 7.5>, angularLimitsMin=<-7.5 -7.5 -7.5>, angularTarget=<1 0 0>}
      localJointOffset=<2.2127 0.1441 2.9271>
     }]
    relativeSpaceBone {boneName=head}
   }
   Output Pose(Result=Component To Local(ComponentPose=@n33)) {layerGroup=None}
  ]}
 EventGraph {nodes [
   Event BlueprintUpdateAnimation()
   > n1 = Gravity Override State Function DEPRECATED(
    CharacterPartAnimInstance=Self-Reference()
    GameTimeInSeconds=GetGameTimeInSeconds()
    GravityOverrideParams=Get GravityOverrideParams() -> (GravityOverrideParams:Gravity Override Params Struct Structure)
   )
   > Set GravityOverride(GravityOverride=@n1)
   Event BlueprintInitializeAnimation()
   > Set Skel_Mesh(Skel_Mesh=GetOwnerSkeletalMeshComponent())
  ]}
]
variables [
 GravityOverride {type=Vector}
 GravityOverrideParams {type=GravityOverrideParamsStruct}
 Skel_Mesh {type=SkeletalMeshComponent}
]
